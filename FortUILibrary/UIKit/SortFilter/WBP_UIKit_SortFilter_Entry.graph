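# Sort/filter option entry widget: a single checkbox-style toggle button backed by a FortSortFilterItemVM.
# Clicking broadcasts OnClicked with the view model, while VM_IsActiveUpdated sets the button's selected state without click feedback.
# GetLabelWithCount appends a bracketed count when ShowCount is on and the count is positive.
# PreConstruct previews the label at design time and sets wrapped, vertically centered text; Construct overrides bottom padding on mobile.
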WidgetBlueprint /FortUILibrary/UIKit/SortFilter/WBP_UIKit_SortFilter_Entry
graphs [
 EventGraph {nodes [
   "On Clicked (WBP_UIKit_ButtonWithToggle)"!() -> (OutputDelegate, then, Button:Common Button Base Object Reference) {is=title K2Node_ComponentBoundEvent}
   > Call On Clicked!(
    execute:Exec
    self:WBP UIKit Sort Filter Entry Object Reference
    SortFilterItemVM:Fort Sort Filter Item VM Object Reference=Get FortSortFilterItemVM() -> (FortSortFilterItemVM:Fort Sort Filter Item VM Object Reference)
   ) -> (then:Exec) {is=title K2Node_CallDelegate}
   n1 = Event PreConstruct()
   > n2 = Sequence()
   > Branch(Condition=@n1.IsDesignTime)
   > SetText!(
    execute
    self:WBP UIKit Button Generic Base Object Reference=Get WBP_UIKit_ButtonWithToggle() -> (WBP_UIKit_ButtonWithToggle:WBP UIKit Button With Toggle Object Reference)
    Text:Text=GetLabelWithCount!(self, Label:Text="LOCTEXT(\"Filter option\")", Count:int=0) -> (ReturnValue:Text)
   ) -> (then)
   SetWrapping!(
    execute=@n2.then_1
    self:WBP UIKit Block Text Object Reference=Get Block_Text!(self:WBP UIKit Button Generic Base Object Reference=Get WBP_UIKit_ButtonWithToggle()) -> (Block_Text:WBP UIKit Block Text Object Reference)
    IsAutoWrap:bool=true
    WrapTextAt:int=0
   ) -> (then)
   n3 = Get WBP_UIKit_ButtonWithToggle()
   > n4 = Cast To GridSlot!(execute, Object=Get Slot!(self:Widget Object Reference=Get Block_Text(self=@n3)) -> (Slot:Panel Slot Object Reference)) -> (then, CastFailed, AsGrid Slot:Grid Slot Object Reference) {is=title K2Node_DynamicCast}
   > SetVerticalAlignment(self:Grid Slot Object Reference=@n4, InVerticalAlignment=VAlign_Center)
   Event Construct()
   > SetMobileLayout()
  ]}
 GetLabelWithCount {nodes [
   n1 = GetLabelWithCount!() -> (then, Label:Text, Count:int) {is=title K2Node_FunctionEntry}
   n2 = "integer > integer"(A=@n1.Count)
   n3 = Get ShowCount() -> (ShowCount:bool)
   > Return Node!(
    execute
    ReturnValue:Text=Select(
     Option 0:Text=@n1
     Option 1:Text=Format Text!(Format="LOCTEXT(\"{Label} [{Count}]\")", Label:Text=@n1, Count:int=@n1.Count) -> (Result)
     Index=AND Boolean(A=@n3, B=@n2)
    ) -> (ReturnValue:Text)
   )
  ]}
 OnClicked {nodes [OnClicked!() -> (then, SortFilterItemVM:Fort Sort Filter Item VM Object Reference) {is=title K2Node_FunctionEntry}]}
 SetMobileLayout {nodes [
   SetMobileLayout() {is=title K2Node_FunctionEntry}
   > Branch(Condition=IsMobileGame())
   n1 = Get!() -> (Padding_Left:float=~, Padding_Top:float=~, Padding_Right:float=~, Padding_Bottom:float=~)
   > SetPadding!(
    execute
    self:User Widget Object Reference=Get WBP_UIKit_ButtonWithToggle()
    InPadding_Left:float=@n1
    InPadding_Top:float=@n1.Padding_Top
    InPadding_Right:float=@n1.Padding_Right
    InPadding_Bottom:float=Get In Padding BottomMobile() -> (In Padding BottomMobile:float)
   ) -> (then)
  ]}
 VM_IsActiveUpdated {nodes [
   n1 = VM_IsActiveUpdated!() -> (then, IsActive:bool=~) {is=title K2Node_FunctionEntry}
   > SetIsSelected!(execute, self:Common Button Base Object Reference=Get WBP_UIKit_ButtonWithToggle(), InSelected:bool=@n1, bGiveClickFeedback:bool=false) -> (then)
  ]}
]
variables [
 In Padding BottomMobile {type=number, default=24}
 OnClicked {}
 ShowCount {type=boolean, default=true}
]
widgets {
 info {
  inheritedWidgetCount=0
  parentClass=&UserWidget
  rootWidgetClass=/FortUILibrary/UIKit/Buttons/WBP_UIKit_ButtonWithToggle.WBP_UIKit_ButtonWithToggle_C
  widgetCount=1
 }
 widgets [WBP_UIKit_ButtonWithToggle {
   widgetClassPath=/FortUILibrary/UIKit/Buttons/WBP_UIKit_ButtonWithToggle
   bIsVariable=true
   widget=WBP_UIKit_ButtonWithToggle
   properties {
    iconHorizontalAlignment=HAlign_Left
    text="Filter Option [0]"
    textJustification=Left
    textSecondJustification=Left
    textTransformPolicy=None
    toggleType=CheckBox
    font {
     bForceMonospaced=false
     bMaterialIsStencil=false
     fontMaterial=/FortUILibrary/UIKit/Buttons/MaterialInstances/GenericButtons/MI_UIKit_ButtonWithToggle_Text
     fontObject=&HeadingNow
     letterSpacing=48
     monospacedWidth=1
     size=18
     skewAmount=0
     typefaceFontName=75 Medium
     outlineSettings {
      bApplyOutlineToDropShadows=false
      bMiteredCorners=false
      bSeparateFillAlpha=false
      outlineMaterial=null
      outlineSize=0
      outlineColor=<rgba 0 0 0 1>
     }
    }
    textPadding=<ltrb 14 0 14 0>
   }
  }]
}
